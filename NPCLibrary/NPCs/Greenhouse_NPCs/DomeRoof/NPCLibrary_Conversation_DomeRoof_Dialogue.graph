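# Conversation database for the DomeRoof Greenhouse NPC, holding a pirate-flavored dialogue about cursed gold.
# The entry tag leads to an "Ask about..." speech node that branches to a "The Curse" topic or a Back choice.
# The curse topic plays a linear chain of speech lines, then a ReturnToStart node labeled BACK restarts the conversation.
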ConversationDatabase /NPCLibrary/NPCs/Greenhouse_NPCs/DomeRoof/NPCLibrary_Conversation_DomeRoof_Dialogue
entry_points [{tag=Conversation.Entry.DomeRoof.Dialogue, destinations [#1]}]
nodes [
 ConversationEntryPointNode_0 {
  id=#1
  class=ConversationEntryPointNode
  outputs [#2]
  properties {entryTag {tagName=Conversation.Entry.DomeRoof.Dialogue}}
 }
 Speech_C_0 {
  id=#2
  class=Speech_C
  outputs [#3, #4]
  properties {
   nodeName=""
   softTaskIcon=&T_UI_NPC_question
   generalConfig {defaultMessage="It be too late to alter course now, matey!"}
  }
  sub_nodes [ConversationChoiceNode_5 {class=ConversationChoiceNode, properties {defaultChoiceDisplayText=Ask about...}}]
 }
 Back_C_0 {
  id=#4
  class=Back_C
  sub_nodes [ConversationChoiceNode_7 {class=ConversationChoiceNode, properties {defaultChoiceDisplayText=BACK}}]
 }
 Speech_C_7 {
  id=#3
  class=Speech_C
  outputs [#5]
  properties {
   nodeName=""
   softTaskIcon=&T_UI_NPC_question
   generalConfig {defaultMessage=A curse was placed on 882 identical gold pieces.}
  }
  sub_nodes [ConversationChoiceNode_15 {class=ConversationChoiceNode, properties {defaultChoiceDisplayText=The Curse}}]
 }
 Speech_C_15 {
  id=#5
  class=Speech_C
  outputs [#6]
  properties {
   nodeName=""
   softTaskIcon=&T_UI_NPC_question
   generalConfig {defaultMessage=Any mortal that removed but a single piece was to be punished for eternity.}
  }
 }
 Speech_C_14 {
  id=#6
  class=Speech_C
  outputs [#7]
  properties {
   nodeName=""
   softTaskIcon=&T_UI_NPC_question
   generalConfig {defaultMessage=I thought I was too old to believe in such tales.}
  }
 }
 Speech_C_1 {
  id=#7
  class=Speech_C
  outputs [#8]
  properties {nodeName="", softTaskIcon=&T_UI_NPC_question, generalConfig {defaultMessage=And that was my mistake.}}
 }
 ReturnToStart_C_23 {
  id=#8
  class=ReturnToStart_C
  properties {softTaskIcon=&T_UI_NPC_Back}
  sub_nodes [ConversationChoiceNode_32 {class=ConversationChoiceNode, properties {defaultChoiceDisplayText=BACK}}]
 }
]
